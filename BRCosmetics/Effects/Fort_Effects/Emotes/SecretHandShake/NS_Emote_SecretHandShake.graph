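# Niagara effect for the Secret Handshake emote: a big golden-orange mesh heart that pops into smaller hearts.
# The heart rises, spins and grows with a flare and glow, then at one second sparkles, smoke, a chromatic ring, second flare and glow, and five falling mini hearts burst out.
# Heart emitters generate location events; both curl-noise ScatterHearts sprite emitters have their own spawn modules disabled.
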NiagaraSystem /BRCosmetics/Effects/Fort_Effects/Emotes/SecretHandShake/NS_Emote_SecretHandShake
dependencies {
 usedDataInterfaces [&NiagaraDataInterfaceCurve, &NiagaraDataInterfaceVector2DCurve, &NiagaraDataInterfaceColorCurve]
 usedDynamicInputs []
 usedModules [&SystemState]
}
emitters {
 BigHeart {
  emitterUpdateScript [
   &EmitterScalability(
    Module.bApplySpawnCountScale=false
    Module.bApplySpawnCountScaleByDistanceFraction=false
    Module.bUseMaxDistance=false
    Module.bUseMinDistance=false
   ) {enabled=false}
   &EmitterLifeCycle() {enabled=false}
   &EmitterState()
   &SpawnBurst_Instantaneous(Spawn Count=1, Use Spawn Probability=false)
  ]
  particleSpawnScript [
   /Niagara/Modules/Spawn/Initialization/InitializeParticle(
    Color=<rgba 2 0.5514 0 1>
    Lifetime=&UniformRangedFloat(Minimum=1.03, Maximum=1.03)
    Mesh Scale=<0.35 0.35 0.35>
    Module.Write Color=true
    Module.Write Lifetime=true
    Module.Write Mass=true
    Module.Write Material Random=true
    Module.Write Position=false
    Module.Write Scale=true
    Module.Write SpriteRotation=false
    Module.Write SpriteSize=false
    Position=&SimulationPosition()
   )
   &AddVelocity(Velocity=<0 0 525>)
   &ApplyInitialForces()
   /Niagara/Modules/Spawn/Orientation/V2/InitialMeshOrientation(Orientation Vector=<0 1 0>, Use Orientation Vector=true, Use Rotation Vector=true)
   &SystemLocation(MaskSpawnGroup=false, Offset=<0 50 50>)
  ]
  particleUpdateScript [
   &ParticleState()
   &Color()
   &DynamicMaterialParameters(
    Index 0 Param 1=&FloatFromCurve(FloatCurve=DataInterface() {propertyValues {Curve {keys [
         {interpMode=RCIM_Cubic, tangentMode=RCTM_User, time=0.1, value=0}
         <at 0.2163 0.6082>
         <at 0.2179 0.6085>
         {interpMode=RCIM_Cubic, tangentMode=RCTM_User, time=0.37, value=1.5}
         {interpMode=RCIM_Cubic, tangentMode=RCTM_User, time=0.6, value=0}
         <at 0.9894 0>
        ]}}})
    Index 0 Param 2=2
    Module.Param0WriteEnabled=true
    Module.Param1WriteEnabled=true
    Module.Param2WriteEnabled=true
    Module.Param3WriteEnabled=true
   )
   &ScaleColor(
    Scale Alpha=&FloatFromCurve(FloatCurve=DataInterface() {propertyValues {Curve {keys [<at 0.2179 0.9943>,<at 0.9412 0.9905>,<at 1.0011 0.002>]}}})
    Scale RGB=&VectorFromFloat(Value=&FloatFromCurve(FloatCurve=DataInterface() {propertyValues {Curve {keys [
          {interpMode=RCIM_Cubic, tangentMode=RCTM_User, time=0.002, value=0.2618}
          {arriveTangent=0.728, interpMode=RCIM_Cubic, leaveTangent=0.728, tangentMode=RCTM_User, time=0.1816, value=1.6275}
          <at 0.2163 1.65>
          <at 0.2179 1.6501>
          <at 0.8429 1.1824>
          {arriveTangent=-1.9835, interpMode=RCIM_Cubic, leaveTangent=-1.9835, tangentMode=RCTM_User, time=0.9893, value=2.6982}
          {interpMode=RCIM_Cubic, tangentMode=RCTM_User, time=0.9991, value=1.0149}
         ]}}}))
    ScaleA=true
    ScaleRGB=true
   )
   &ScaleMeshSize(Scale Factor=&ScaleFloatByCurve(
     InitialValue:Vector3f=&VectorFromFloat(Value=1)
     ScaleCurve:NiagaraDataInterfaceVectorCurve=DataInterface() {propertyValues {
       XCurve {keys [
         {arriveTangent=6.8907, interpMode=RCIM_Cubic, leaveTangent=6.8907, tangentMode=RCTM_User, time=0, value=0}
         {arriveTangent=2, interpMode=RCIM_Cubic, leaveTangent=2, tangentMode=RCTM_User, time=0.3958, value=2.3319}
         {arriveTangent=2.8264, interpMode=RCIM_Cubic, leaveTangent=2.8264, tangentMode=RCTM_User, time=0.8126, value=2.5809}
         {interpMode=RCIM_Cubic, tangentMode=RCTM_User, time=0.9902, value=4.0121}
        ]}
       YCurve {keys [
         {arriveTangent=5.7969, interpMode=RCIM_Cubic, leaveTangent=5.7969, tangentMode=RCTM_User, time=0, value=0}
         {arriveTangent=2, interpMode=RCIM_Cubic, leaveTangent=2, tangentMode=RCTM_User, time=0.3958, value=2.3319}
         {arriveTangent=2.8264, interpMode=RCIM_Cubic, leaveTangent=2.8264, tangentMode=RCTM_User, time=0.8126, value=2.5809}
         {interpMode=RCIM_Cubic, tangentMode=RCTM_User, time=0.9902, value=4.0121}
        ]}
       ZCurve {keys [
         {arriveTangent=5.9063, interpMode=RCIM_Cubic, leaveTangent=5.9063, tangentMode=RCTM_User, time=0, value=0}
         {arriveTangent=2, interpMode=RCIM_Cubic, leaveTangent=2, tangentMode=RCTM_User, time=0.3958, value=2.3319}
         {arriveTangent=2.8264, interpMode=RCIM_Cubic, leaveTangent=2.8264, tangentMode=RCTM_User, time=0.8126, value=2.5809}
         {interpMode=RCIM_Cubic, tangentMode=RCTM_User, time=0.9902, value=4.0121}
        ]}
      }}
     CurveIndex:NiagaraFloat=$Particles.NormalizedAge
    ))
   &Drag(Drag=8, Use Linear Drag=true, Use Rotational Drag=false)
   &SolveForcesAndVelocity(Clamp Velocity=false, Limit Acceleration=false)
   &GenerateLocationEvent(Module.Use Event Send Rate=true)
   /Niagara/Modules/Update/Orientation/V2/UpdateMeshOrientation(
    Rotation Rate:NiagaraFloat=&Multiply_Float(
     A=$Particles.Age
     B=&FloatFromCurve(FloatCurve=DataInterface() {propertyValues {Curve {keys [<at 0 0>,<at 1 -3>]}}})
    )
    Rotation Vector:Vector3f=<0 0 1>
   )
  ]
  rendererClasses [&NiagaraMeshRendererProperties]
  renderers [
   {rendererClass=&NiagaraMeshRendererProperties, rendererIndex=0}
   {rendererClass=&NiagaraMeshRendererProperties, rendererIndex=1}
  ]
 }
 BigHeartSparkles {
  emitterUpdateScript [
   &EmitterScalability(
    Module.bApplySpawnCountScale=false
    Module.bApplySpawnCountScaleByDistanceFraction=false
    Module.bUseMaxDistance=false
    Module.bUseMinDistance=false
   ) {enabled=false}
   &EmitterLifeCycle() {enabled=false}
   &EmitterState()
   &SpawnBurst_Instantaneous(Spawn Count=1, Spawn Time=1, Use Spawn Probability=false)
  ]
  particleSpawnScript [
   &InitializeParticle(
    Color=<rgba 2 0.5514 0 1>
    Lifetime=&UniformRangedFloat(Minimum=1.25, Maximum=1.25)
    Mesh Scale=<3 3 3>
    Module.Write Color=true
    Module.Write Lifetime=true
    Module.Write Mass=true
    Module.Write Material Random=true
    Module.Write Position=false
    Module.Write Scale=true
    Module.Write SpriteRotation=false
    Module.Write SpriteSize=false
    Position=&SimulationPosition()
   )
   &AddVelocity(Velocity=<0 0 1500>) {enabled=false}
   &ApplyInitialForces()
   &InitialMeshOrientation(Use Orientation Vector=true, Use Rotation Vector=true)
   &SystemLocation(MaskSpawnGroup=false, Offset=<0 50 100>)
  ]
  particleUpdateScript [
   &ParticleState()
   &Color(Color=<rgba 15 1 1 1>) {enabled=false}
   &DynamicMaterialParameters(
    Index 0 Param 1=&FloatFromCurve(FloatCurve=DataInterface() {propertyValues {Curve {keys [{interpMode=RCIM_Cubic, tangentMode=RCTM_User, time=0.1, value=15}, <at 0.9601 33.4933>]}}})
    Index 0 Param 2=&FloatFromCurve(FloatCurve=DataInterface() {propertyValues {Curve {keys [{interpMode=RCIM_Cubic, tangentMode=RCTM_User, time=0, value=0.5}, <at 0.9019 3.9428>]}}})
    Index 0 Param 3=1
    Module.Param0WriteEnabled=true
    Module.Param1WriteEnabled=true
    Module.Param2WriteEnabled=true
    Module.Param3WriteEnabled=true
   )
   &ScaleColor(
    Scale Alpha=&FloatFromCurve(FloatCurve=DataInterface() {propertyValues {Curve {keys [
         {interpMode=RCIM_Cubic, tangentMode=RCTM_User, time=-0.0045, value=1.0052}
         {arriveTangent=-1.2671, interpMode=RCIM_Cubic, leaveTangent=-1.2671, tangentMode=RCTM_User, time=0.3587, value=0.2516}
         <at 1.0011 0.002>
        ]}}})
    Scale RGB=/Niagara/DynamicInputs/ScaleByCurve/ScaleVectorByCurve(
     InitialValue:Vector3f=<1 1 1>
     ScaleCurve:NiagaraDataInterfaceVectorCurve=DataInterface() {propertyValues {
       XCurve {keys [
         {interpMode=RCIM_Cubic, tangentMode=RCTM_User, time=0, value=150}
         {arriveTangent=-138.2454, interpMode=RCIM_Cubic, leaveTangent=-138.2427, tangentMode=RCTM_User, time=0.2, value=23}
         {arriveTangent=-48.375, interpMode=RCIM_Cubic, leaveTangent=-48.375, tangentMode=RCTM_User, time=0.4, value=8.3}
         {interpMode=RCIM_Cubic, tangentMode=RCTM_User, time=1, value=0}
        ]}
       YCurve {keys [
         {interpMode=RCIM_Cubic, tangentMode=RCTM_User, time=0, value=150}
         {arriveTangent=-143.3364, interpMode=RCIM_Cubic, leaveTangent=-143.3364, tangentMode=RCTM_User, time=0.2, value=20.3}
         {arriveTangent=-45, interpMode=RCIM_Cubic, leaveTangent=-45, tangentMode=RCTM_User, time=0.4, value=5.3}
         {interpMode=RCIM_Cubic, tangentMode=RCTM_User, time=1, value=0}
        ]}
       ZCurve {keys [
         {interpMode=RCIM_Cubic, tangentMode=RCTM_User, time=0, value=150}
         {arriveTangent=-143.3364, interpMode=RCIM_Cubic, leaveTangent=-143.3364, tangentMode=RCTM_User, time=0.2, value=21}
         {arriveTangent=-45.875, interpMode=RCIM_Cubic, leaveTangent=-45.875, tangentMode=RCTM_User, time=0.4, value=6}
         {interpMode=RCIM_Cubic, tangentMode=RCTM_User, time=1, value=0}
        ]}
      }}
     CurveIndex:NiagaraFloat=$Particles.NormalizedAge
    )
    ScaleA=true
    ScaleRGB=true
   )
   &ScaleMeshSize(Scale Factor=&ScaleFloatByCurve(
     InitialValue:Vector3f=&VectorFromFloat(Value=1)
     ScaleCurve:NiagaraDataInterfaceVectorCurve=DataInterface() {propertyValues {
       XCurve {keys [
         {arriveTangent=1.6491, interpMode=RCIM_Cubic, leaveTangent=1.6491, tangentMode=RCTM_User, time=0.001, value=1.01}
         {arriveTangent=2.7945, interpMode=RCIM_Cubic, leaveTangent=2.7945, tangentMode=RCTM_User, time=0.1641, value=1.754}
         {interpMode=RCIM_Cubic, tangentMode=RCTM_User, time=0.9959, value=2.5102}
        ]}
       YCurve {keys [
         {arriveTangent=1.5861, interpMode=RCIM_Cubic, leaveTangent=1.5861, tangentMode=RCTM_User, time=0.001, value=1.01}
         {arriveTangent=2.7848, interpMode=RCIM_Cubic, leaveTangent=2.7848, tangentMode=RCTM_User, time=0.1613, value=1.7709}
         {interpMode=RCIM_Cubic, tangentMode=RCTM_User, time=0.9959, value=2.5102}
        ]}
       ZCurve {keys [
         {arriveTangent=3678.8083, interpMode=RCIM_Cubic, leaveTangent=6.2126, tangentMode=RCTM_User, time=0.001, value=1.01}
         {arriveTangent=2.1063, interpMode=RCIM_Cubic, leaveTangent=2.1063, tangentMode=RCTM_User, time=0.1594, value=1.8556}
         {interpMode=RCIM_Cubic, tangentMode=RCTM_User, time=0.9959, value=2.5102}
        ]}
      }}
     CurveIndex:NiagaraFloat=$Particles.NormalizedAge
    ))
   &Drag(Drag=8, Use Linear Drag=true, Use Rotational Drag=false) {enabled=false}
   &SolveForcesAndVelocity(Clamp Velocity=false, Limit Acceleration=false) {enabled=false}
   &UpdateMeshOrientation(
    Rotation Rate:NiagaraFloat=&Multiply_Float(
     A=$Particles.Age
     B=&FloatFromCurve(FloatCurve=DataInterface() {propertyValues {Curve {keys [<at -0.019 -10>,<at 1.0011 0.002>]}}})
    )
    Rotation Vector:Vector3f=<0 0 1>
   ) {enabled=false}
  ]
  rendererClasses [&NiagaraMeshRendererProperties]
  renderers [
   {rendererClass=&NiagaraMeshRendererProperties, rendererIndex=0}
   {rendererClass=&NiagaraMeshRendererProperties, rendererIndex=1}
  ]
 }
 ChromaticRing {
  emitterUpdateScript [
   &EmitterScalability(
    Module.bApplySpawnCountScale=false
    Module.bApplySpawnCountScaleByDistanceFraction=false
    Module.bUseMaxDistance=false
    Module.bUseMinDistance=false
   ) {enabled=false}
   &EmitterLifeCycle(MaxLoopCount=1, NextLoopDuration=0.1) {enabled=false}
   &EmitterState()
   &SpawnBurst_Instantaneous(Spawn Count=1, Spawn Probability=0, Use Spawn Probability=true)
   &SpawnBurst_Instantaneous(Spawn Count=1, Spawn Time=1, Use Spawn Probability=false) {module=SpawnBurst_Instantaneous001}
  ]
  particleSpawnScript [
   &InitializeParticle(
    Color=<rgba 50 13.3333 0 1>
    Lifetime=&Lerp_Float(A=0.25, B=0.35, Alpha=$System.Age)
    Module.Write Color=true
    Module.Write Lifetime=true
    Module.Write Mass=true
    Module.Write Material Random=true
    Module.Write Position=true
    Module.Write Scale=false
    Module.Write SpriteRotation=true
    Module.Write SpriteSize=true
    Position=&SimulationPosition()
    Sprite Rotation=&UniformRangedFloat(Minimum=0, Maximum=360)
    Sprite Size=&Vector2DFromFloat(Value=&UniformRangedFloat(Minimum=525, Maximum=700))
   )
   &SystemLocation(
    MaskSpawnGroup=false
    Offset=&VectorFromCurve(
     VectorCurve=DataInterface() {propertyValues {XCurve {keys [<at 0 0>,<at 0.5 0>]}, YCurve {keys [<at 0 50>,<at 0.5 50>]}, ZCurve {keys [<at 0 50>,<at 0.5 120>]}}}
     CurveIndex=$Emitter.Age
    )
   )
  ]
  particleUpdateScript [
   &ParticleState()
   &SolveForcesAndVelocity(Clamp Velocity=false, Limit Acceleration=false)
   &ScaleColor(
    Scale Alpha=&FloatFromCurve(FloatCurve=DataInterface() {propertyValues {Curve {keys [<at -0.0046 1.0113>,<at 1.0011 0.002>]}}})
    Scale RGB=&VectorFromCurve(VectorCurve=DataInterface() {propertyValues {
       XCurve {keys [<at -0.0011 1.0067>,<at 0.9988 0>]}
       YCurve {keys [<at -0.0011 1.0067>,<at 0.9988 0>]}
       ZCurve {keys [<at -0.0011 1.0067>,<at 0.9988 0>]}
      }})
    ScaleA=true
    ScaleRGB=true
   )
   &DynamicMaterialParameters(
    Index 0 Param 1=&FloatFromCurve(FloatCurve=DataInterface() {propertyValues {Curve {keys [<at 0.9171 1.4301>,<at 1 1>]}}})
    Index 0 Param 3=1
    Index 0 Param 4=1
    Module.Param0WriteEnabled=true
    Module.Param1WriteEnabled=true
    Module.Param2WriteEnabled=true
    Module.Param3WriteEnabled=true
   )
   &ScaleSpriteSize(Scale Factor=&Vector2DFromCurve(
     Vector2Curve=DataInterface() {propertyValues {XCurve {keys [<at -0.0107 0.2143>,<at 1.0179 0.6584>]}, YCurve {keys [<at -0.0269 0.1988>,<at 1.0179 0.6584>]}}}
     ScaleCurve=<1 1>
    ))
   &SpriteRotationRate(Rotation Rate=&UniformRangedFloat(Minimum=-50, Maximum=50))
  ]
  rendererClasses [&NiagaraSpriteRendererProperties]
  renderers [{rendererClass=&NiagaraSpriteRendererProperties}]
 }
 Flare {
  emitterUpdateScript [
   &EmitterScalability(
    Module.bApplySpawnCountScale=false
    Module.bApplySpawnCountScaleByDistanceFraction=false
    Module.bUseMaxDistance=false
    Module.bUseMinDistance=false
   ) {enabled=false}
   &EmitterLifeCycle(MaxLoopCount=1, NextLoopDuration=0.1) {enabled=false}
   &EmitterState()
   &SpawnBurst_Instantaneous(Spawn Count=1, Use Spawn Probability=false)
   &SpawnBurst_Instantaneous(Spawn Count=1, Spawn Time=1, Use Spawn Probability=false) {module=SpawnBurst_Instantaneous001}
  ]
  particleSpawnScript [
   &InitializeParticle(
    Color=<rgba 2 0.9261 0 1>
    Lifetime=&Lerp_Float(A=0.25, B=0.5, Alpha=$System.Age)
    Module.Write Color=true
    Module.Write Lifetime=true
    Module.Write Mass=true
    Module.Write Material Random=true
    Module.Write Position=true
    Module.Write Scale=false
    Module.Write SpriteRotation=true
    Module.Write SpriteSize=true
    Position=&SimulationPosition()
    Sprite Rotation=&UniformRangedFloat(Minimum=0, Maximum=360)
    Sprite Size=&Vector2DFromFloat(Value=&UniformRangedFloat(Minimum=140, Maximum=175))
   )
   &SystemLocation(
    MaskSpawnGroup=false
    Offset=&VectorFromCurve(
     VectorCurve=DataInterface() {propertyValues {XCurve {keys [<at 0 0>,<at 0.5 0>]}, YCurve {keys [<at 0 50>,<at 0.5 50>]}, ZCurve {keys [<at 0 50>,<at 0.5 120>]}}}
     CurveIndex=$Emitter.Age
    )
   )
   /SetVariables(Particles.Color=&LerpLinearColors(StartColor=<rgba 2 0.9261 0 1>, EndColor=<rgba 5 1.6667 0 1>, LerpFactor=$System.Age)) {isSetParameters=true}
  ]
  particleUpdateScript [
   &ParticleState()
   &SolveForcesAndVelocity(Clamp Velocity=false, Limit Acceleration=false)
   &ScaleSpriteSize(Scale Factor=&Vector2DFromCurve(
     Vector2Curve=DataInterface() {propertyValues {XCurve {keys [<at 0.0019 1.9456>,<at 0.9988 0>]}, YCurve {keys [<at 0.0019 1.9456>,<at 0.9988 0>]}}}
     ScaleCurve=<1 1>
    ))
   &ScaleColor(
    Scale Alpha=&FloatFromCurve(FloatCurve=DataInterface() {propertyValues {Curve {keys [<at -0.0046 1.0113>,<at 1.0011 0.002>]}}})
    Scale RGB=&VectorFromCurve(VectorCurve=DataInterface() {propertyValues {
       XCurve {keys [<at -0.0011 1.0067>,<at 0.9988 0>]}
       YCurve {keys [<at -0.0011 1.0067>,<at 0.9988 0>]}
       ZCurve {keys [<at -0.0011 1.0067>,<at 0.9988 0>]}
      }})
    ScaleA=true
    ScaleRGB=true
   )
  ]
  rendererClasses [&NiagaraSpriteRendererProperties]
  renderers [{rendererClass=&NiagaraSpriteRendererProperties}]
 }
 Glow {
  emitterUpdateScript [
   &EmitterScalability(
    Module.bApplySpawnCountScale=false
    Module.bApplySpawnCountScaleByDistanceFraction=false
    Module.bUseMaxDistance=false
    Module.bUseMinDistance=false
   ) {enabled=false}
   &EmitterLifeCycle() {enabled=false}
   &EmitterState()
   &SpawnBurst_Instantaneous(Spawn Count=1, Use Spawn Probability=false)
   &SpawnBurst_Instantaneous(Spawn Count=1, Spawn Time=1, Use Spawn Probability=false) {module=SpawnBurst_Instantaneous001}
  ]
  particleSpawnScript [
   &InitializeParticle(
    Color=<rgba 2500 2500 1 1>
    Lifetime=&UniformRangedFloat(Minimum=0.5, Maximum=1)
    Mesh Scale=<0.25 0.25 0.25>
    Module.Write Color=true
    Module.Write Lifetime=true
    Module.Write Mass=true
    Module.Write Material Random=true
    Module.Write Position=false
    Module.Write Scale=true
    Module.Write SpriteRotation=false
    Module.Write SpriteSize=false
    Position=&SimulationPosition()
   )
   &SystemLocation(
    MaskSpawnGroup=false
    Offset=&VectorFromCurve(
     VectorCurve=DataInterface() {propertyValues {XCurve {keys [<at 0 0>,<at 0.5 0>]}, YCurve {keys [<at 0 50>,<at 0.5 50>]}, ZCurve {keys [<at 0 50>,<at 0.5 120>]}}}
     CurveIndex=$Emitter.Age
    )
   )
  ]
  particleUpdateScript [
   &ParticleState()
   &Color(Color=<rgba 5000 2500 10 1>)
   &DynamicMaterialParameters(
    Index 0 Param 1=&FloatFromCurve(FloatCurve=DataInterface() {propertyValues {Curve {keys [<at -0.0005 -0.0141>,<at 0.1666 0.3179>,<at 1.0021 1.9784>,<at 1.61 1.9784>]}}})
    Index 0 Param 2=&FloatFromCurve(FloatCurve=DataInterface() {propertyValues {Curve {keys [<at -0.0005 15>,<at 0.1612 13.3874>,<at 1.0021 5>,<at 1.61 5>]}}})
    Index 0 Param 3=-0.8
    Module.Param0WriteEnabled=true
    Module.Param1WriteEnabled=true
    Module.Param2WriteEnabled=true
    Module.Param3WriteEnabled=true
   )
   &ScaleMeshSize(Scale Factor=&VectorFromCurve(VectorCurve=DataInterface() {propertyValues {
       XCurve {keys [
         {interpMode=RCIM_Cubic, tangentMode=RCTM_User, time=-0.0004, value=0.0471}
         {arriveTangent=24.8774, interpMode=RCIM_Cubic, leaveTangent=24.8774, tangentMode=RCTM_User, time=0.1612, value=4.1479}
         {arriveTangent=6.8631, interpMode=RCIM_Cubic, leaveTangent=6.8631, tangentMode=RCTM_User, time=0.2646, value=6.6389}
         <at 0.9996 7.8749>
         <at 1.61 7.8749>
        ]}
       YCurve {keys [
         {interpMode=RCIM_Cubic, tangentMode=RCTM_User, time=-0.0004, value=0.0471}
         {arriveTangent=24.5017, interpMode=RCIM_Cubic, leaveTangent=24.5017, tangentMode=RCTM_User, time=0.1612, value=4.0789}
         {arriveTangent=6.8631, interpMode=RCIM_Cubic, leaveTangent=6.8631, tangentMode=RCTM_User, time=0.2651, value=6.5525}
         <at 0.9996 7.8749>
         <at 1.61 7.8749>
        ]}
       ZCurve {keys [
         {arriveTangent=25.8257, interpMode=RCIM_Cubic, leaveTangent=25.8257, tangentMode=RCTM_User, time=-0.0004, value=0.0471}
         <at 0.1612 4.8326>
         {arriveTangent=6.8631, interpMode=RCIM_Cubic, leaveTangent=6.8631, tangentMode=RCTM_User, time=0.2666, value=6.7585}
         <at 0.9996 7.8749>
         <at 1.61 7.8749>
        ]}
      }}))
   &ScaleColor(
    Scale Alpha=&FloatFromCurve(FloatCurve=DataInterface() {propertyValues {Curve {keys [
         {interpMode=RCIM_Cubic, tangentMode=RCTM_User, time=-0.0034, value=0.1203}
         {arriveTangent=-0.1606, interpMode=RCIM_Cubic, leaveTangent=-0.1606, tangentMode=RCTM_User, time=0.3345, value=0.0495}
         {interpMode=RCIM_Cubic, tangentMode=RCTM_User, time=0.9972, value=-0.0001}
        ]}}})
    Scale RGB=&ScaleFloatByCurve(
     InitialValue:Vector3f=<1 1 1>
     ScaleCurve:NiagaraDataInterfaceVectorCurve=DataInterface() {propertyValues {
       XCurve {keys [<at 0.0013 1.998>,<at 1 1>]}
       YCurve {keys [<at 0.0013 0.4577>,<at 1.0017 0.264>]}
       ZCurve {keys [<at -0.0045 0.4024>,<at 1.0017 0.264>]}
      }}
     CurveIndex:NiagaraFloat=$Particles.NormalizedAge
    )
    ScaleA=true
    ScaleRGB=true
   )
  ]
  rendererClasses [&NiagaraMeshRendererProperties]
  renderers [{rendererClass=&NiagaraMeshRendererProperties, rendererIndex=0}]
 }
 ScatterHearts {
  emitterUpdateScript [
   &EmitterState(Life Cycle Mode=::NewEnumerator1, Loop Behavior=::NewEnumerator1, Loop Duration=1.765, UseLoopDelay=false)
   &SpawnBurst_Instantaneous(Use Spawn Probability=false) {enabled=false}
   &SpawnRate(SpawnRate=150, Use Spawn Probability=false) {enabled=false}
  ]
  particleSpawnScript [
   &InitializeParticle(
    Color=<rgba 1 0.2757 0 1>
    Lifetime=&UniformRangedFloat(Minimum=0.5, Maximum=1)
    Module.Write Color=true
    Module.Write Lifetime=true
    Module.Write Mass=true
    Module.Write Material Random=true
    Module.Write Position=false
    Module.Write Scale=false
    Module.Write SpriteRotation=true
    Module.Write SpriteSize=true
    Position=&SimulationPosition()
    Sprite Rotation=0.5
    Sprite Size=&UniformRangedVector2D(Minimum=<17.5 8.75>, Maximum=<17.5 8.75>)
   )
   /Niagara/Modules/Update/Mass/CalculateSizeAndRotationalInertiaByMass(Depth=0, Height=5, Module.Manually Enter Density=false) {enabled=false}
   &AddVelocity(Velocity=<500 0 0>) {enabled=false}
   /SetVariables(Particles.Color_Sparkles:LinearColor=<rgba 25 15 1 0.5>, Particles.SpriteSize_Sparkles:Vector2f=<25 25>) {isSetParameters=true, module=SetVariables_#1}
  ]
  particleUpdateScript [
   &ParticleState(Kill Particles When Lifetime Has Elapsed=false, Let Infinitely Lived Particles Die When Emitter Deactivates=true)
   &GravityForce(Gravity=&MakeVector(
     X=&UniformRangedFloat(Minimum=-150, Maximum=150)
     Y=&UniformRangedFloat(Minimum=-250, Maximum=250)
     Z=&FloatFromCurve(FloatCurve=DataInterface() {propertyValues {Curve {keys [{interpMode=RCIM_Cubic, tangentMode=RCTM_User, time=0, value=0}, {interpMode=RCIM_Cubic, tangentMode=RCTM_User, time=1, value=-980}]}}})
    ))
   &Drag(Drag=&UniformRangedFloat(Minimum=1, Maximum=5), Use Linear Drag=true, Use Rotational Drag=false)
   &ScaleColor(
    Scale Alpha=&FloatFromCurve(FloatCurve=DataInterface() {propertyValues {Curve {keys [<at -0.0031 0.1837 RCIM_Cubic>,<at 1.0011 0.002 RCIM_Cubic>]}}})
    ScaleA=true
    ScaleRGB=true
   )
   /Niagara/Modules/Update/Forces/CurlNoiseForce(Noise Frequency=2, Noise Quality / Cost=::NewEnumerator0, Noise Strength=1000, Pan Noise=true)
   /SetVariables(
    Particles.Color_Sparkles:LinearColor=&ScaleLinearColorByCurve(
     InitialValue=<rgba 1 0.5573 0 0.25>
     ScaleCurve=DataInterface() {propertyValues {
       AlphaCurve {keys [<at 0 15>,<at 1.0491 0>]}
       BlueCurve {keys [<at 0 0>,<at 0.3965 1.391>]}
       GreenCurve {keys [<at 0 56.8485>,<at 0.3965 4.2489>]}
       RedCurve {keys [<at 0 150>,<at 0.3965 15>]}
      }}
     CurveIndex=$Particles.NormalizedAge
    )
    Particles.SpriteSize_Sparkles:Vector2f=&UniformRangedVector2D(Minimum=<10 10>, Maximum=<50 50>)
   ) {isSetParameters=true, module=SetVariables_#2}
   &SolveForcesAndVelocity(Clamp Velocity=false, Limit Acceleration=false)
   &ScaleSpriteSize(Scale Factor=&Vector2DFromCurve(
     Vector2Curve=DataInterface() {propertyValues {
       XCurve {keys [
         {interpMode=RCIM_Cubic, tangentMode=RCTM_User, time=-0.0034, value=1.9443}
         {arriveTangent=-1.1505, interpMode=RCIM_Cubic, leaveTangent=-1.1506, tangentMode=RCTM_User, time=0.495, value=1.0094}
         {interpMode=RCIM_Cubic, tangentMode=RCTM_User, time=0.9937, value=0.2566}
        ]}
       YCurve {keys [
         {interpMode=RCIM_Cubic, tangentMode=RCTM_User, time=-0.0034, value=1.9443}
         {arriveTangent=-1.3287, interpMode=RCIM_Cubic, leaveTangent=-1.3287, tangentMode=RCTM_User, time=0.5, value=1}
         {interpMode=RCIM_Cubic, tangentMode=RCTM_User, time=0.9937, value=0.2566}
        ]}
      }}
     ScaleCurve=<0.5 0.5>
    ))
   &ScaleSpriteSizeBySpeed(
    Max Scale Factor=<1 2>
    Sample Scale Factor By Curve=true
    Scale Factor Curve=DataInterface() {propertyValues {Curve {keys [
        {arriveTangent=2.865, interpMode=RCIM_Cubic, time=0.1247, value=-0.0006}
        {arriveTangent=4.1522, interpMode=RCIM_Cubic, leaveTangent=4.1522, tangentMode=RCTM_User, time=0.3948, value=0.4893}
        <at 0.9958 0.9974 RCIM_Cubic>
       ]}}}
    Velocity Threshold=150
   ) {module=SpriteSizeScaleByVelocity}
  ]
  rendererClasses [&NiagaraSpriteRendererProperties]
  renderers [{rendererClass=&NiagaraSpriteRendererProperties}]
 }
 ScatterHearts001 {
  emitterUpdateScript [
   &EmitterState(Life Cycle Mode=::NewEnumerator1, Loop Behavior=::NewEnumerator1, Loop Duration=1.765, UseLoopDelay=false)
   &SpawnBurst_Instantaneous(Use Spawn Probability=false) {enabled=false}
   &SpawnRate(SpawnRate=150, Use Spawn Probability=false) {enabled=false}
  ]
  particleSpawnScript [
   &InitializeParticle(
    Color=<rgba 1 0.2757 0 1>
    Lifetime=&UniformRangedFloat(Minimum=0.5, Maximum=1)
    Module.Write Color=true
    Module.Write Lifetime=true
    Module.Write Mass=true
    Module.Write Material Random=true
    Module.Write Position=false
    Module.Write Scale=false
    Module.Write SpriteRotation=true
    Module.Write SpriteSize=true
    Position=&SimulationPosition()
    Sprite Rotation=0.5
    Sprite Size=&UniformRangedVector2D(Minimum=<17.5 8.75>, Maximum=<17.5 8.75>)
   )
   &CalculateSizeAndRotationalInertiaByMass(Depth=0, Height=5, Module.Manually Enter Density=false) {enabled=false}
   &AddVelocity(Velocity=<500 0 0>) {enabled=false}
   /SetVariables(Particles.Color_Sparkles:LinearColor=<rgba 25 5 1 0.5>, Particles.SpriteSize_Sparkles:Vector2f=<10 10>) {isSetParameters=true, module=SetVariables_#1}
  ]
  particleUpdateScript [
   &ParticleState(Kill Particles When Lifetime Has Elapsed=false, Let Infinitely Lived Particles Die When Emitter Deactivates=true)
   &GravityForce(Gravity=&MakeVector(
     X=&UniformRangedFloat(Minimum=-150, Maximum=150)
     Y=&UniformRangedFloat(Minimum=-250, Maximum=250)
     Z=&FloatFromCurve(FloatCurve=DataInterface() {propertyValues {Curve {keys [{interpMode=RCIM_Cubic, tangentMode=RCTM_User, time=0, value=0}, {interpMode=RCIM_Cubic, tangentMode=RCTM_User, time=1, value=-980}]}}})
    ))
   &Drag(Drag=&UniformRangedFloat(Minimum=1, Maximum=5), Use Linear Drag=true, Use Rotational Drag=false)
   &ScaleColor(
    Scale Alpha=&FloatFromCurve(FloatCurve=DataInterface() {propertyValues {Curve {keys [<at -0.0031 0.1837 RCIM_Cubic>,<at 1.0011 0.002 RCIM_Cubic>]}}})
    ScaleA=true
    ScaleRGB=true
   )
   &CurlNoiseForce(Noise Frequency=2, Noise Quality / Cost=::NewEnumerator0, Noise Strength=1000, Pan Noise=true, Pan Noise Field=<0 0 1>)
   &SolveForcesAndVelocity(Clamp Velocity=false, Limit Acceleration=false)
   &ScaleSpriteSize(Scale Factor=&Vector2DFromCurve(
     Vector2Curve=DataInterface() {propertyValues {
       XCurve {keys [
         {interpMode=RCIM_Cubic, tangentMode=RCTM_User, time=-0.0034, value=1.9443}
         {arriveTangent=-1.1505, interpMode=RCIM_Cubic, leaveTangent=-1.1506, tangentMode=RCTM_User, time=0.5, value=0.5}
         {interpMode=RCIM_Cubic, tangentMode=RCTM_User, time=0.9937, value=0.2566}
        ]}
       YCurve {keys [
         {interpMode=RCIM_Cubic, tangentMode=RCTM_User, time=-0.0034, value=1.9443}
         {arriveTangent=-1.3287, interpMode=RCIM_Cubic, leaveTangent=-1.3287, tangentMode=RCTM_User, time=0.5, value=0.5}
         {interpMode=RCIM_Cubic, tangentMode=RCTM_User, time=0.9937, value=0.2566}
        ]}
      }}
     ScaleCurve=<0.5 0.5>
    ))
   /SetVariables(
    Particles.Color_Sparkles:LinearColor=&ScaleLinearColorByCurve(
     InitialValue=<rgba 1 0.5573 0 0.25>
     ScaleCurve=DataInterface() {propertyValues {
       AlphaCurve {keys [<at 0 15>,<at 1.0491 0>]}
       BlueCurve {keys [<at 0 0>,<at 0.3965 1.391>]}
       GreenCurve {keys [<at 0 56.8485>,<at 0.3965 4.2489>]}
       RedCurve {keys [<at 0 150>,<at 0.3965 15>]}
      }}
     CurveIndex=$Particles.NormalizedAge
    )
    Particles.SpriteSize_Sparkles:Vector2f=&UniformRangedVector2D(Minimum=<15 15>, Maximum=<25 25>)
   ) {isSetParameters=true, module=SetVariables_#2}
  ]
  rendererClasses [&NiagaraSpriteRendererProperties]
  renderers [{rendererClass=&NiagaraSpriteRendererProperties}]
 }
 SmallerHearts {
  emitterUpdateScript [
   &EmitterScalability(
    Module.bApplySpawnCountScale=false
    Module.bApplySpawnCountScaleByDistanceFraction=false
    Module.bUseMaxDistance=false
    Module.bUseMinDistance=false
   ) {enabled=false}
   &EmitterLifeCycle() {enabled=false}
   &EmitterState()
   &SpawnBurst_Instantaneous(Spawn Count=5, Spawn Time=1, Use Spawn Probability=false)
  ]
  particleSpawnScript [
   &InitializeParticle(
    Color=<rgba 2 0.5514 0 1>
    Lifetime=&UniformRangedFloat(Minimum=0.7, Maximum=1)
    Mesh Scale=<0.21 0.21 0.21>
    Module.Write Color=true
    Module.Write Lifetime=true
    Module.Write Mass=true
    Module.Write Material Random=true
    Module.Write Position=false
    Module.Write Scale=true
    Module.Write SpriteRotation=false
    Module.Write SpriteSize=false
    Position=&SimulationPosition()
   )
   /Niagara/Modules/Spawn/Location/SphereLocation(MaskSpawnGroup=false, Module.Override Randomness=false, Sphere Radius=25, Surface Only=false) {enabled=false}
   &InitialMeshOrientation(
    Orientation Vector=/Niagara/DynamicInputs/Random/RandomVector(
     Vector Scale:NiagaraFloat=1
     Constraint Vector:Vector3f=<0 0 1>
     Randomness Mode:ENiagaraRandomnessMode=::NewEnumerator0
     Random Seed:NiagaraInt32=0
     Fixed Random Seed:NiagaraBool=false
    )
    Use Orientation Vector=true
    Use Rotation Vector=false
   )
   &AddVelocityInCone(Cone Angle=90, Cone Axis=<0 0 1>, Module.Use Velocity Falloff On Cone Axis=false, Velocity Strength=400)
   &AddVelocity(Velocity=&UniformRangedVector(Minimum=<-35 -35 35>, Maximum=<35 35 350>))
   &AccelerationForce(Acceleration=<0 0 1>)
   &ApplyInitialForces()
   &SystemLocation(MaskSpawnGroup=false, Offset=<0 50 100>)
  ]
  particleUpdateScript [
   &ParticleState()
   &Color()
   &DynamicMaterialParameters(
    Index 0 Param 1=&FloatFromCurve(FloatCurve=DataInterface() {propertyValues {Curve {keys [
         {interpMode=RCIM_Cubic, tangentMode=RCTM_User, time=0.1, value=0}
         <at 0.2163 0.6082>
         <at 0.2179 0.6085>
         {interpMode=RCIM_Cubic, tangentMode=RCTM_User, time=0.37, value=1.5}
         {interpMode=RCIM_Cubic, tangentMode=RCTM_User, time=0.6, value=0}
         <at 0.9894 0>
        ]}}})
    Index 0 Param 2=2
    Module.Param0WriteEnabled=true
    Module.Param1WriteEnabled=true
    Module.Param2WriteEnabled=true
    Module.Param3WriteEnabled=true
   )
   &ScaleColor(
    Scale Alpha=&FloatFromCurve(FloatCurve=DataInterface() {propertyValues {Curve {keys [<at 0.2179 0.9943>,<at 0.6249 0.9943>,<at 0.9894 0.0329>,<at 1.0011 0.002>]}}})
    Scale RGB=&VectorFromFloat(Value=&FloatFromCurve(FloatCurve=DataInterface() {propertyValues {Curve {keys [
          {interpMode=RCIM_Cubic, tangentMode=RCTM_User, time=-0.0088, value=2.4259}
          <at 0.4824 1.1097>
          {arriveTangent=-1.9835, interpMode=RCIM_Cubic, leaveTangent=-1.9835, tangentMode=RCTM_User, time=0.936, value=0.9416}
          <at 0.9894 0.0787>
          {interpMode=RCIM_Cubic, tangentMode=RCTM_User, time=1.0011, value=0.002}
         ]}}}))
    ScaleA=true
    ScaleRGB=true
   )
   &ScaleMeshSize(Scale Factor=&ScaleFloatByCurve(
     InitialValue:Vector3f=&VectorFromFloat(Value=1)
     ScaleCurve:NiagaraDataInterfaceVectorCurve=DataInterface() {propertyValues {
       XCurve {keys [
         {interpMode=RCIM_Cubic, tangentMode=RCTM_User, time=0, value=2}
         {arriveTangent=-1, interpMode=RCIM_Cubic, leaveTangent=-1, tangentMode=RCTM_User, time=0.5, value=2}
         {interpMode=RCIM_Cubic, tangentMode=RCTM_User, time=1, value=0}
        ]}
       YCurve {keys [
         {interpMode=RCIM_Cubic, tangentMode=RCTM_User, time=0, value=2}
         {arriveTangent=-1, interpMode=RCIM_Cubic, leaveTangent=-1, tangentMode=RCTM_User, time=0.5, value=2}
         {interpMode=RCIM_Cubic, tangentMode=RCTM_User, time=1, value=0}
        ]}
       ZCurve {keys [
         {interpMode=RCIM_Cubic, tangentMode=RCTM_User, time=0, value=2}
         {arriveTangent=-1, interpMode=RCIM_Cubic, leaveTangent=-1, tangentMode=RCTM_User, time=0.5, value=2}
         {interpMode=RCIM_Cubic, tangentMode=RCTM_User, time=1, value=0}
        ]}
      }}
     CurveIndex:NiagaraFloat=$Particles.NormalizedAge
    ))
   &MeshRotationForce(Rotation=&UniformRangedVector(Minimum=<-15 -15 -15>, Maximum=<25 25 25>))
   &AccelerationForce(Acceleration=<0 0 -1600>) {module=AccelerationForce001}
   &SolveForcesAndVelocity(Clamp Velocity=false, Limit Acceleration=false)
   &SolveRotationalForcesAndVelocity()
   &GenerateLocationEvent(Module.Use Event Send Rate=true)
  ]
  rendererClasses [&NiagaraMeshRendererProperties]
  renderers [{rendererClass=&NiagaraMeshRendererProperties, rendererIndex=0}]
 }
 Smoke {
  emitterUpdateScript [
   &EmitterScalability(
    Module.bApplySpawnCountScale=false
    Module.bApplySpawnCountScaleByDistanceFraction=false
    Module.bUseMaxDistance=false
    Module.bUseMinDistance=false
   ) {enabled=false}
   &EmitterLifeCycle() {enabled=false}
   &EmitterState()
   &SpawnBurst_Instantaneous(Spawn Count=1, Spawn Time=1, Use Spawn Probability=false)
  ]
  particleSpawnScript [
   &InitializeParticle(
    Color=<rgba 1 0.3806 0 1>
    Lifetime=&UniformRangedFloat(Minimum=0.5, Maximum=0.5)
    Module.Write Color=true
    Module.Write Lifetime=true
    Module.Write Mass=true
    Module.Write Material Random=true
    Module.Write Position=false
    Module.Write Scale=false
    Module.Write SpriteRotation=true
    Module.Write SpriteSize=true
    Position=&SimulationPosition()
    Sprite Rotation=&UniformRangedFloat(Minimum=0, Maximum=360)
    Sprite Size=&Vector2DFromFloat(Value=&UniformRangedFloat(Minimum=175, Maximum=175))
   )
   &SystemLocation(
    MaskSpawnGroup=false
    Offset=&VectorFromCurve(
     VectorCurve=DataInterface() {propertyValues {XCurve {keys [<at 0 0>,<at 0.5 0>]}, YCurve {keys [<at 0 50>,<at 0.5 50>]}, ZCurve {keys [<at 0 50>,<at 0.5 120>]}}}
     CurveIndex=$Emitter.Age
    )
   )
  ]
  particleUpdateScript [
   &ParticleState()
   &SpriteRotationRate(Rotation Rate=&Multiply_Float(
     A=$Particles.InitSpriteRotRate
     B=&FloatFromCurve(FloatCurve=DataInterface() {propertyValues {Curve {keys [<at -0.0046 1.0113 RCIM_Cubic>,<at 1.0011 0.09 RCIM_Cubic>]}}})
    )) {enabled=false}
   &ScaleSpriteSize(Scale Factor=&Vector2DFromFloat(Value=&FloatFromCurve(FloatCurve=DataInterface() {propertyValues {Curve {keys [
          {arriveTangent=2.9149, interpMode=RCIM_Cubic, leaveTangent=2.9149, tangentMode=RCTM_User, time=-0.003, value=0}
          {arriveTangent=0.9127, interpMode=RCIM_Cubic, leaveTangent=0.9127, tangentMode=RCTM_User, time=0.23, value=0.63}
          <at 1 1>
         ]}}})))
   &CameraOffset(Camera Offset Amount=20)
   &SolveForcesAndVelocity(Clamp Velocity=false, Limit Acceleration=false)
   &DynamicMaterialParameters(
    Index 0 Param 1=1
    Module.Param0WriteEnabled=true
    Module.Param1WriteEnabled=true
    Module.Param2WriteEnabled=true
    Module.Param3WriteEnabled=true
   )
   /SetVariables(Particles.SubImageIndex=&Lerp_Float(A=16, B=63, Alpha=$Particles.NormalizedAge)) {isSetParameters=true}
  ]
  rendererClasses [&NiagaraSpriteRendererProperties]
  renderers [{rendererClass=&NiagaraSpriteRendererProperties}]
 }
}
summary {
 systemName=NS_Emote_SecretHandShake
 emitters [
  {emitterName=BigHeart, rendererClasses [&NiagaraMeshRendererProperties]}
  {emitterName=SmallerHearts, rendererClasses [&NiagaraMeshRendererProperties]}
  {emitterName=Flare, rendererClasses [&NiagaraSpriteRendererProperties]}
  {emitterName=ChromaticRing, rendererClasses [&NiagaraSpriteRendererProperties]}
  {emitterName=ScatterHearts, rendererClasses [&NiagaraSpriteRendererProperties]}
  {emitterName=Smoke, rendererClasses [&NiagaraSpriteRendererProperties]}
  {emitterName=BigHeartSparkles, rendererClasses [&NiagaraMeshRendererProperties]}
  {emitterName=Glow, rendererClasses [&NiagaraMeshRendererProperties]}
  {emitterName=ScatterHearts001, rendererClasses [&NiagaraSpriteRendererProperties]}
 ]
 userVariables [User.ScaleCurve.Scale Factor {
   type=NiagaraDataInterfaceVectorCurve
   description=""
   defaultValue {
    struct=NiagaraExt_VariableValue_DataInterface
    value {
     dataInterface=&self:NiagaraDataInterfaceVectorCurve_0
     dataInterfaceClass=&NiagaraDataInterfaceVectorCurve
    }
   }
  }]
}
